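# Ambient prop Sound Cue for a looping console/ATM screen buzz, built from reused Foley hologram loop waves.
# A Random node picks one of three equally weighted looping waves, feeding a Modulator and then a Mixer that halves its volume.
# Inline overrides replace the referenced attenuation and concurrency assets, giving a short-range capsule falloff with listener focus and a small voice limit.
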SoundCue /Game/Sounds/Helios/Environment_Actors/ConsoleATM/Spot_Console_Screen_Buzz_Capsule
root=Mixer_0
nodes [
 {id=Mixer_0, children [Modulator_1,null], properties {inputVolume [0.5,1]}}
 {id=Modulator_1, children [Random_1], properties {pitchMin=1}}
 {id=Random_1, children [WavePlayer_0,WavePlayer_5,WavePlayer_6], properties {weights [1,1,1]}}
 {
  id=WavePlayer_0
  properties {bLooping=true, soundWaveAssetPtr=/Game/Sounds/Foley_Loot/Containers/Keys_And_Locks/Keys_And_Locks_HologramLoop_01}
 }
 {
  id=WavePlayer_5
  properties {bLooping=true, soundWaveAssetPtr=/Game/Sounds/Foley_Loot/Containers/Keys_And_Locks/Keys_And_Locks_HologramLoop_02}
 }
 {
  id=WavePlayer_6
  properties {bLooping=true, soundWaveAssetPtr=/Game/Sounds/Foley_Loot/Containers/Keys_And_Locks/Keys_And_Locks_HologramLoop_03}
 }
]
properties {
 attenuationSettings=/Game/Sounds/Helios/Environment_Actors/LaptopScreenStatic/Spot_Laptop_Screen_Static_Sphere_Att
 bOverrideAttenuation=true
 bOverrideConcurrency=true
 maxDistance=453
 soundClassObject=&Fort_Ambient_Props_Athena
 volumeMultiplier=1
 attenuationOverrides {
  attenuationShape=Capsule
  bEnableListenerFocus=true
  falloffDistance=350
  focusAzimuth=25
  focusVolumeAttenuation=1.2
  nonFocusAzimuth=55
  nonFocusVolumeAttenuation=0.75
  attenuationShapeExtents=<103 50>
 }
 concurrencyOverrides {maxCount=6}
 concurrencySet [&Weapon_Drop_StopOldest]
}
